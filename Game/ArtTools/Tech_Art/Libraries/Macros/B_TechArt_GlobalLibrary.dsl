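; Tech-art global library of shared helpers for actor tags, bounds overlap tests, dynamic material instances and scalability settings update binding.
; The Houdini builder refills an instanced static mesh from a data table, sampling every Nth row, swapping Y and Z, with optional random rotation and scale.
; Lumen and Battle Royale volumetric cloud checks combine platform tests with user settings or a console variable, and the Lumen check logs its result.
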(fn AddActorTag
  (params (Actor :type Actor%20Object%20Reference) (Tag :type Name%20%28by%20ref%29))
  (bind _output_get (LocalVariables|SetExistingActorTags (Variables|/Script/Engine.Actor|GetTags Actor)))
  (Utilities|Array|AddUnique _output_get Tag)
  (Variables|/Script/Engine.Actor|SetTags (LocalVariables|GetExistingActorTags) Actor))

(fn AreBoundsOverlapping?
  (params (BoundsA_Min_X :type Float%20%28double-precision%29) (BoundsA_Min_Y :type Float%20%28double-precision%29) (BoundsA_Min_Z :type Float%20%28double-precision%29) (BoundsA_Max_X :type Float%20%28double-precision%29) (BoundsA_Max_Y :type Float%20%28double-precision%29) (BoundsA_Max_Z :type Float%20%28double-precision%29) (BoundsA_IsValid :type Boolean) (BoundsB_Min_X :type Float%20%28double-precision%29) (BoundsB_Min_Y :type Float%20%28double-precision%29) (BoundsB_Min_Z :type Float%20%28double-precision%29) (BoundsB_Max_X :type Float%20%28double-precision%29) (BoundsB_Max_Y :type Float%20%28double-precision%29) (BoundsB_Max_Z :type Float%20%28double-precision%29) (BoundsB_IsValid :type Boolean))
  (return (Math|Boolean|ANDBoolean (and (> BoundsA_Max_X BoundsB_Min_X) (> BoundsB_Max_X BoundsA_Min_X)) (and (> BoundsA_Max_Y BoundsB_Min_Y) (> BoundsB_Max_Y BoundsA_Min_Y)) (and (> BoundsA_Max_Z BoundsB_Min_Z) (> BoundsB_Max_Z BoundsA_Min_Z)))))

(fn BindScalabilitySettings
  (params (Delegate :type Delegate%20%28by%20ref%29))
  (EventDispatchers|BindEventtoOnGameUserSettingsUINeedsUpdate (Settings|GetFortGameUserSettings) Delegate))

(fn GetDynamicMaterialInstance
  (params (MeshComponent :type Mesh%20Component%20Object%20Reference) (MaterialIndex :type Integer))
  (bind _returnvalue (Rendering|Material|GetMaterial MeshComponent MaterialIndex))
  (bind _asmaterial_instance_dynamic (Utilities|Casting|CastToMaterialInstanceDynamic _returnvalue)
    (:then
      (return _asmaterial_instance_dynamic))
    (:CastFailed
      (bind _returnvalue_2 (Rendering|Material|CreateDynamicMaterialInstance MeshComponent MaterialIndex _returnvalue "None" :!class /Script/Engine.PrimitiveComponent))
      (return _returnvalue_2))))

(fn HoudiniInstanceHISMBuilder
  (params (InstancedStaticMeshComponent :type Instanced%20Static%20Mesh%20Component%20Object%20Reference) (InstanceDataTable :type Data%20Table%20Object%20Reference) (InstancesCountToImport :type Float%20%28double-precision%29 :default 1.0) (RandomScale :type Boolean) (RandomScaleMin :type Float%20%28double-precision%29) (RandomScaleMax :type Float%20%28double-precision%29 :default 1.0) (RandomRotationZ :type Boolean) (RandomRotationZMin :type Float%20%28double-precision%29 :default -5e+01) (RandomRotationZMax :type Float%20%28double-precision%29 :default 5e+01) (UseWorldSpacePositions :type Boolean))
  (Utilities|IsValid InstancedStaticMeshComponent
    (:Is%20Valid
      (Components|InstancedStaticMesh|ClearInstances (Variables|GetInstancedStaticMeshComponent))
      (bind _outrownames (DataTable|GetDataTableRowNames (Variables|GetInstanceDataTable)))
      (for (_array_element _array_index) _outrownames
        (if (== (Math|Integer|%25%28Integer%29 _array_index (Math|Float|Truncate (/ 1.0 (Variables|GetInstancesCountToImport)))) 0)
          (bind _returnvalue (Utilities|GetDataTableRow (Variables|GetInstanceDataTable) _array_element))
          (Components|InstancedStaticMesh|AddInstance (Variables|GetInstancedStaticMeshComponent) (Math|Transform|MakeTransform (Math|Vector|MakeVector (Utilities|Struct|BreakHoudiniInstanceDataLayout _returnvalue) (.Pz_8_488530224541EBF7D43F2F99AE2CCEAF (Utilities|Struct|BreakHoudiniInstanceDataLayout _returnvalue)) (.Py_7_462CEB644FCF8818190E0DA7A4233236 (Utilities|Struct|BreakHoudiniInstanceDataLayout _returnvalue))) (Math|Rotator|SelectRotator (Math|Rotator|MakeRotator 0.0 0.0 (Math|Random|RandomFloatinRange (Variables|GetRandomRotationZMin) (Variables|GetRandomRotationZMax))) "0, 0, 0" (Variables|GetRandomRotationZ)) (Math|Vector|SelectVector (Math|Vector|MakeVector (Math|Random|RandomFloatinRange (Variables|GetRandomScaleMin) (Variables|GetRandomScaleMax)) (Math|Random|RandomFloatinRange (Variables|GetRandomScaleMin) (Variables|GetRandomScaleMax)) (Math|Random|RandomFloatinRange (Variables|GetRandomScaleMin) (Variables|GetRandomScaleMax))) (Math|Vector|MakeVector (.pscale_24_C1C793994E739F1A41356ABC35528C18 (Utilities|Struct|BreakHoudiniInstanceDataLayout _returnvalue)) (.pscale_24_C1C793994E739F1A41356ABC35528C18 (Utilities|Struct|BreakHoudiniInstanceDataLayout _returnvalue)) (.pscale_24_C1C793994E739F1A41356ABC35528C18 (Utilities|Struct|BreakHoudiniInstanceDataLayout _returnvalue))) (Variables|GetRandomScale))) (Variables|GetUseWorldSpacePositions)))))))

(fn IsLumenEnabled
  (bind _returnvalue (FortGlobals|IsRunninginCloud))
  (bind _returnvalue_2 (TouchPlatform|IsMobileGame 0))
  (bind _returnvalue_3 (not _returnvalue_2))
  (bind _returnvalue_4 (TouchPlatform|IsNintendoSwitch))
  (bind _returnvalue_5 (not _returnvalue_4))
  (bind _returnvalue_6 (or _returnvalue_3 _returnvalue))
  (bind _returnvalue_7 (Settings|GetGameUserSettings))
  (Fortnite|FortShippingLog 0 (Utilities|String|Append "TODM Scalability: Is Lumen Supported = " (Utilities|String|ToString(Boolean) (Math|Boolean|ANDBoolean (Settings|IsGlobalIlluminationAllowed _returnvalue_7 0) _returnvalue_5 _returnvalue_6))))
  (return (Math|Boolean|ANDBoolean (Settings|IsGlobalIlluminationAllowed _returnvalue_7 0) _returnvalue_5 _returnvalue_6)))

(fn UseVolumetricCloudsBR
  (bind _returnvalue (Development|GetConsoleVariableIntValue "Athena.VolumetricClouds.Quality"))
  (return (and (> _returnvalue 0) (or (not (TouchPlatform|IsNintendoSwitch)) (not (TouchPlatform|IsMobileGame 0))))))
